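# Masked, unlit material for the Pineapple Bandit female medium body, named as a "no hair" variant, likely to hide the hair.
# It is flagged for skeletal meshes and mesh deformers.
# The graph holds only a zero constant and world position multiplied by -1, a pattern typical of masking out and collapsing geometry.
# The file does not show which material outputs these nodes feed.
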Material /BRCosmetics/Characters/Player/Female/Medium/Bodies/F_MED_Pineapple_Bandit/Materials/F_MED_PineappleBandit_HairNone
graph {
 properties {
  bUseTranslucencyVertexFog=false
  bUsedWithMeshDeformer=true
  bUsedWithSkeletalMesh=true
  blendMode=BLEND_Masked
  pixelDepthOffsetMode=PDOM_Legacy
  shadingModel=MSM_Unlit
  usedShadingModels=Unlit
 }
 nodes [
  Constant()
  Multiply(A=WorldPosition().XYZ) {constB=-1}
 ]
}
