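# Physics asset for the Pineapple Bandit female outfit's chest belt, defining secondary-motion bodies for the dangling belt and its two grenades.
# A capsule belt chain hangs from spine_02, with a wider swing axis at its top and a tight joint between segments.
# Both grenades branch from the lower segment with limited swing and twist, one through a base body.
# Large pelvis and spine capsules approximate the torso, while their joints and the base-body attachments rotate freely.
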PhysicsAsset /BRCosmetics/Characters/Player/Female/Medium/Bodies/F_MED_Pineapple_Bandit/Meshes/F_MED_Pineapple_Bandit_ChestBelt_Physics
bodies {
 dyn_chest_belt_a [{shapeType=Capsule, length=6, radius=3.5, center=<6.8697 0 0>, rotation=<rot 90 0 0>}]
 dyn_chest_belt_a_base [{shapeType=Capsule, length=1, radius=1}]
 dyn_chest_belt_b [{shapeType=Capsule, length=7.75, radius=2.25, center=<0.4697 0 0>, rotation=<rot 0 0 90>}]
 dyn_grenade_a [{shapeType=Capsule, length=3, radius=3.75, center=<2.7055 0.6779 2.0616>, rotation=<rot 82.8857 -176.1857 -179.5541>}]
 dyn_grenade_a_base [{shapeType=Capsule, length=1, radius=1}]
 dyn_grenade_b [{shapeType=Capsule, length=3, radius=3.75, center=<1.7835 0 3.0766>, rotation=<rot 65.6924 -180 -180>}]
 pelvis [{shapeType=Capsule, length=9, radius=10.5, center=<0.973 -1.6187 0>}]
 spine_01 [{shapeType=Capsule, length=5.25, radius=9.75, center=<0 0.3491 0>}]
 spine_02 [{shapeType=Capsule, length=0.25, radius=10.25, center=<0 0.8286 0>}]
}
constraints [
 {bone1Name=dyn_chest_belt_a, bone2Name=dyn_chest_belt_a_base, swing1LimitDegrees=5, swing2LimitDegrees=45, twistLimitDegrees=5}
 {
  bone1Name=dyn_grenade_a_base
  bone2Name=dyn_chest_belt_b
  swing1LimitDegrees=45
  swing1Motion=Free
  swing2LimitDegrees=45
  swing2Motion=Free
  twistLimitDegrees=45
  twistMotion=Free
 }
 {bone1Name=dyn_grenade_a, bone2Name=dyn_grenade_a_base, swing1LimitDegrees=25, swing2LimitDegrees=25, twistLimitDegrees=10}
 {bone1Name=dyn_grenade_b, bone2Name=dyn_chest_belt_b, swing1LimitDegrees=25, swing2LimitDegrees=25, twistLimitDegrees=10}
 {bone1Name=dyn_chest_belt_b, bone2Name=dyn_chest_belt_a, swing1LimitDegrees=5, swing2LimitDegrees=5, twistLimitDegrees=5}
 {
  bone1Name=dyn_chest_belt_a_base
  bone2Name=spine_02
  swing1LimitDegrees=45
  swing1Motion=Free
  swing2LimitDegrees=45
  swing2Motion=Free
  twistLimitDegrees=45
  twistMotion=Free
 }
 {
  bone1Name=spine_01
  bone2Name=pelvis
  swing1LimitDegrees=45
  swing1Motion=Free
  swing2LimitDegrees=45
  swing2Motion=Free
  twistLimitDegrees=45
  twistMotion=Free
 }
 {
  bone1Name=spine_02
  bone2Name=spine_01
  swing1LimitDegrees=45
  swing1Motion=Free
  swing2LimitDegrees=45
  swing2Motion=Free
  twistLimitDegrees=45
  twistMotion=Free
 }
]
