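# Physics asset for the Pineapple Bandit female medium face accessory (jacket variant), giving bodies to its dynamic strands, ribbon and collar bones.
# Capsules cover the head, neck, clavicles and upper spine, while five angled boxes form the collar pieces.
# Three two-segment strands and a two-sided ribbon branch from head-attached base bodies with limited swing and twist.
# The base bodies, collar boxes and core body links all leave swing and twist free.
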PhysicsAsset /BRCosmetics/Characters/Player/Female/Medium/Bodies/F_MED_Pineapple_Bandit/Meshes/Parts/F_MED_Pineapple_Bandit_FaceAcc_Jacket_Physics
bodies {
 clavicle_l [{shapeType=Capsule, length=9.25, radius=6.25, center=<3.62 -0.9816 -0.3333>, rotation=<rot 75.9422 0 0>}]
 clavicle_r [{shapeType=Capsule, length=9.25, radius=6.25, center=<-3.6199 0.8332 0.3333>, rotation=<rot 75.9414 0 180>}]
 dyn_collar_back [{shapeType=Box, extentX=11, extentY=9.5, extentZ=2, center=<-0.1283 0 -0.2591>, rotation=<rot -26.3559 0 0>}]
 dyn_collar_front_l [{shapeType=Box, extentX=8, extentY=11.5, extentZ=2.5, rotation=<rot -34.3722 9.3477 -14.6505>}]
 dyn_collar_front_r [{shapeType=Box, extentX=6.5, extentY=11.5, extentZ=2.5, rotation=<rot -26.573 -4.0679 -15.5796>}]
 dyn_collar_side_l [{shapeType=Box, extentX=8.5, extentY=11, extentZ=2, center=<-0.1268 0.0484 -0.2996>, rotation=<rot -26.5204 -7.1775 -1.4122>}]
 dyn_collar_side_r [{shapeType=Box, extentX=9, extentY=11, extentZ=2.5, rotation=<rot -22.0859 -9.0982 3.4456>}]
 dyn_ribbon [{shapeType=Capsule, length=1, radius=1}]
 dyn_ribbon_l [{shapeType=Capsule, length=3.5898, radius=2.018, center=<1.4143 -2.3194 1.7031>, rotation=<rot 0 47.6773 -68.3484>}]
 dyn_ribbon_r [{shapeType=Capsule, length=3.7948, radius=2.0729, center=<-1.1811 -3.5747 1.1368>, rotation=<rot 0 -7.0045 -51.9465>}]
 dyn_strandA_a [{shapeType=Capsule, length=4.75, radius=1, center=<3.4193 0.0551 0>, rotation=<rot 90 0 -5.97>}]
 dyn_strandA_a_base [{shapeType=Capsule, length=0.505, radius=0.505, rotation=<rot 0 -90 0>}]
 dyn_strandA_b [{shapeType=Capsule, length=5.25, radius=0.75, center=<3.2363 0 0>, rotation=<rot 90 0 0>}]
 dyn_strandB_a [{shapeType=Capsule, length=2.4354, radius=1.7866, center=<1.9696 -0.04 0.0533>, rotation=<rot 0.1076 94.0432 -86.4843>}]
 dyn_strandB_a_base [{shapeType=Capsule, length=0.5, radius=0.5}]
 dyn_strandB_b [{shapeType=Capsule, length=2.9579, radius=2.2603, center=<3.0544 0.0061 0.1916>, rotation=<rot 0 87.5799 -73.4884>}]
 dyn_strandC_a [{shapeType=Capsule, length=2.2178, radius=0.8958, center=<2.0887 0.0246 0.0222>, rotation=<rot 0 -89.4561 -89.1971>}]
 dyn_strandC_a_base [{shapeType=Capsule, length=0, radius=0.8}]
 dyn_strandC_b [{shapeType=Capsule, length=3.3673, radius=1.0976, center=<2.473 -0.0206 -0.0813>, rotation=<rot 0.1561 87.5728 -88.292>}]
 head [{shapeType=Capsule, length=5.25, radius=9.25, center=<3.6495 0.5989 0.0001>, rotation=<rot 90 0.0066 0.0066>}]
 neck_01 [{shapeType=Capsule, length=8.25, radius=6.25, center=<-0.3898 1.4176 0>, rotation=<rot 90 0 0>}]
 spine_05 [{shapeType=Capsule, length=13.25, radius=8.25, center=<2.8263 -1.1791 -0.0001>}]
}
constraints [
 {
  bone1Name=head
  bone2Name=neck_01
  swing1LimitDegrees=45
  swing1Motion=Free
  swing2LimitDegrees=45
  swing2Motion=Free
  twistLimitDegrees=45
  twistMotion=Free
 }
 {
  bone1Name=clavicle_l
  bone2Name=spine_05
  swing1LimitDegrees=45
  swing1Motion=Free
  swing2LimitDegrees=45
  swing2Motion=Free
  twistLimitDegrees=45
  twistMotion=Free
 }
 {
  bone1Name=neck_01
  bone2Name=spine_05
  swing1LimitDegrees=45
  swing1Motion=Free
  swing2LimitDegrees=45
  swing2Motion=Free
  twistLimitDegrees=45
  twistMotion=Free
 }
 {
  bone1Name=dyn_strandC_a_base
  bone2Name=head
  swing1LimitDegrees=45
  swing1Motion=Free
  swing2LimitDegrees=45
  swing2Motion=Free
  twistLimitDegrees=45
  twistMotion=Free
 }
 {bone1Name=dyn_strandC_a, bone2Name=dyn_strandC_a_base, swing1LimitDegrees=45, swing2LimitDegrees=45, twistLimitDegrees=5}
 {bone1Name=dyn_strandC_b, bone2Name=dyn_strandC_a, swing1LimitDegrees=45, swing2LimitDegrees=45, twistLimitDegrees=5}
 {
  bone1Name=dyn_strandB_a_base
  bone2Name=head
  swing1LimitDegrees=45
  swing1Motion=Free
  swing2LimitDegrees=45
  swing2Motion=Free
  twistLimitDegrees=45
  twistMotion=Free
 }
 {bone1Name=dyn_strandB_a, bone2Name=dyn_strandB_a_base, swing1LimitDegrees=45, swing2LimitDegrees=45, twistLimitDegrees=5}
 {bone1Name=dyn_strandB_b, bone2Name=dyn_strandB_a, swing1LimitDegrees=45, swing2LimitDegrees=45, twistLimitDegrees=5}
 {
  bone1Name=dyn_strandA_a_base
  bone2Name=head
  swing1LimitDegrees=45
  swing1Motion=Free
  swing2LimitDegrees=45
  swing2Motion=Free
  twistLimitDegrees=45
  twistMotion=Free
 }
 {bone1Name=dyn_strandA_a, bone2Name=dyn_strandA_a_base, swing1LimitDegrees=25, swing2LimitDegrees=25, twistLimitDegrees=5}
 {bone1Name=dyn_strandA_b, bone2Name=dyn_strandA_a, swing1LimitDegrees=25, swing2LimitDegrees=25, twistLimitDegrees=5}
 {
  bone1Name=clavicle_r
  bone2Name=spine_05
  swing1LimitDegrees=45
  swing1Motion=Free
  swing2LimitDegrees=45
  swing2Motion=Free
  twistLimitDegrees=45
  twistMotion=Free
 }
 {
  bone1Name=dyn_ribbon
  bone2Name=head
  swing1LimitDegrees=45
  swing1Motion=Free
  swing2LimitDegrees=45
  swing2Motion=Free
  twistLimitDegrees=45
  twistMotion=Free
 }
 {bone1Name=dyn_ribbon_l, bone2Name=dyn_ribbon, swing1LimitDegrees=15, swing2LimitDegrees=15, twistLimitDegrees=15}
 {bone1Name=dyn_ribbon_r, bone2Name=dyn_ribbon, swing1LimitDegrees=15, swing2LimitDegrees=15, twistLimitDegrees=25}
 {
  bone1Name=dyn_collar_front_l
  bone2Name=spine_05
  swing1LimitDegrees=45
  swing1Motion=Free
  swing2LimitDegrees=45
  swing2Motion=Free
  twistLimitDegrees=45
  twistMotion=Free
 }
 {
  bone1Name=dyn_collar_side_l
  bone2Name=spine_05
  swing1LimitDegrees=45
  swing1Motion=Free
  swing2LimitDegrees=45
  swing2Motion=Free
  twistLimitDegrees=45
  twistMotion=Free
 }
 {
  bone1Name=dyn_collar_front_r
  bone2Name=spine_05
  swing1LimitDegrees=45
  swing1Motion=Free
  swing2LimitDegrees=45
  swing2Motion=Free
  twistLimitDegrees=45
  twistMotion=Free
 }
 {
  bone1Name=dyn_collar_side_r
  bone2Name=spine_05
  swing1LimitDegrees=45
  swing1Motion=Free
  swing2LimitDegrees=45
  swing2Motion=Free
  twistLimitDegrees=45
  twistMotion=Free
 }
 {
  bone1Name=dyn_collar_back
  bone2Name=spine_05
  swing1LimitDegrees=45
  swing1Motion=Free
  swing2LimitDegrees=45
  swing2Motion=Free
  twistLimitDegrees=45
  twistMotion=Free
 }
]
