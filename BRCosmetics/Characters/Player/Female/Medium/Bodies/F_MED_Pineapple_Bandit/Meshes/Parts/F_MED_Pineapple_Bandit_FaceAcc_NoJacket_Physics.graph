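# Physics asset for the no-jacket face accessory part of the female medium Pineapple Bandit cosmetic, covering dynamic strand and ribbon bones.
# Capsules on the upper spine, clavicles, neck and head form the body shapes.
# Strand base and ribbon root bodies join the head with free angular constraints, while their child segments carry tighter per-chain swing and twist limits.
# The MainB strand forks into two E segments, and the ribbon splits left and right.
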PhysicsAsset /BRCosmetics/Characters/Player/Female/Medium/Bodies/F_MED_Pineapple_Bandit/Meshes/Parts/F_MED_Pineapple_Bandit_FaceAcc_NoJacket_Physics
bodies {
 clavicle_l [{shapeType=Capsule, length=9.25, radius=6.25, center=<3.62 -0.9816 -0.3333>, rotation=<rot 75.9422 0 0>}]
 clavicle_r [{shapeType=Capsule, length=9.25, radius=6.25, center=<-3.6199 0.8332 0.3333>, rotation=<rot 75.9414 0 180>}]
 dyn_ribbon [{shapeType=Capsule, length=1, radius=1}]
 dyn_ribbon_l [{shapeType=Capsule, length=3.5898, radius=2.018, center=<1.4143 -2.3194 1.7031>, rotation=<rot 0 47.6773 -68.3484>}]
 dyn_ribbon_r [{shapeType=Capsule, length=3.7948, radius=2.0729, center=<-1.1811 -3.5747 1.1368>, rotation=<rot 0 -7.0045 -51.9465>}]
 dyn_strandA_a [{shapeType=Capsule, length=3.75, radius=2, center=<2.8927 0 0>, rotation=<rot 90 0 -5.97>}]
 dyn_strandA_a_base [{shapeType=Capsule, length=0.505, radius=0.505, rotation=<rot 0 -90 0>}]
 dyn_strandA_b [{shapeType=Capsule, length=1.75, radius=2.75, center=<2.2541 0 -0.0001>, rotation=<rot 90 0 0>}]
 dyn_strandB_a [{shapeType=Capsule, length=1.1854, radius=3.0366, center=<1.9696 -0.04 0.0533>, rotation=<rot 0.1076 94.0432 -86.4843>}]
 dyn_strandB_a_base [{shapeType=Capsule, length=0.5, radius=0.5}]
 dyn_strandB_b [{shapeType=Capsule, length=1.4579, radius=3.7603, center=<3.0544 0.0061 0.1916>, rotation=<rot 0 87.5799 -73.4884>}]
 dyn_strandC_a [{shapeType=Capsule, length=0.4678, radius=2.6458, center=<2.0887 0.0246 0.0222>, rotation=<rot 0 -89.4561 -89.1971>}]
 dyn_strandC_a_base [{shapeType=Capsule, length=0, radius=0.8}]
 dyn_strandC_b [{shapeType=Capsule, length=1.6173, radius=2.8476, center=<2.473 -0.0206 -0.0813>, rotation=<rot 0.1561 87.5728 -88.292>}]
 dyn_strandD_a [{shapeType=Capsule, length=2.75, radius=2.75, center=<3.0851 0.0001 0.1519>, rotation=<rot 83.6609 180 180>}]
 dyn_strandD_a_base [{shapeType=Capsule, length=0.5, radius=0.5}]
 dyn_strandD_b [{shapeType=Capsule, length=2.25, radius=2, center=<2.3489 0.0001 0>, rotation=<rot 90 0 -3.9935>}]
 dyn_strandE_ca [{shapeType=Capsule, length=4.7959, radius=1.1135, center=<3.2114 -0.1767 0.4282>, rotation=<rot 0 79.9354 -76.9842>}]
 dyn_strandE_cb [{shapeType=Capsule, length=2.9156, radius=1.0857, center=<2.1052 -0.5131 0.0356>, rotation=<rot 0 71.9546 -85.7779>}]
 dyn_strandF_a [{shapeType=Capsule, length=5.5, radius=1.25, center=<3.8864 -0.5388 0.8248>, rotation=<rot 79.1001 179.9999 -174.0906>}]
 dyn_strandF_a_base [{shapeType=Capsule, length=0.5, radius=0.5}]
 dyn_strandF_b [{shapeType=Capsule, length=3.5, radius=1.25, center=<3.0476 0 0>, rotation=<rot 90 0 0>}]
 dyn_strandMainB_a [{shapeType=Capsule, length=3.25, radius=1.25, center=<2.5534 -0.1691 0>, rotation=<rot 90 35.5588 35.9577>}]
 dyn_strandMainB_a_base [{shapeType=Capsule, length=0.5, radius=0.5}]
 dyn_strandMainB_b [{shapeType=Capsule, length=3.6621, radius=1.818, center=<2.8269 0.0783 0.3411>, rotation=<rot 0 94.5125 -81.5222>}]
 head [{shapeType=Capsule, length=7, radius=7.5, center=<3.6495 0.5989 0.0001>, rotation=<rot 90 0.0066 0.0066>}]
 neck_01 [{shapeType=Capsule, length=8.25, radius=6.25, center=<-0.3898 1.4176 0>, rotation=<rot 90 0 0>}]
 neck_02 [{shapeType=Capsule, length=15.755, radius=7.755, center=<-1.0033 4.1516 0.0001>, rotation=<rot 0 -90 0>}]
 spine_05 [{shapeType=Capsule, length=13.25, radius=8.25, center=<2.8263 -1.1791 -0.0001>}]
}
constraints [
 {
  bone1Name=head
  bone2Name=neck_02
  swing1LimitDegrees=45
  swing1Motion=Free
  swing2LimitDegrees=45
  swing2Motion=Free
  twistLimitDegrees=45
  twistMotion=Free
 }
 {
  bone1Name=clavicle_l
  bone2Name=spine_05
  swing1LimitDegrees=45
  swing1Motion=Free
  swing2LimitDegrees=45
  swing2Motion=Free
  twistLimitDegrees=45
  twistMotion=Free
 }
 {
  bone1Name=neck_01
  bone2Name=spine_05
  swing1LimitDegrees=45
  swing1Motion=Free
  swing2LimitDegrees=45
  swing2Motion=Free
  twistLimitDegrees=45
  twistMotion=Free
 }
 {
  bone1Name=dyn_strandC_a_base
  bone2Name=head
  swing1LimitDegrees=45
  swing1Motion=Free
  swing2LimitDegrees=45
  swing2Motion=Free
  twistLimitDegrees=45
  twistMotion=Free
 }
 {bone1Name=dyn_strandC_a, bone2Name=dyn_strandC_a_base, swing1LimitDegrees=45, swing2LimitDegrees=45, twistLimitDegrees=5}
 {bone1Name=dyn_strandC_b, bone2Name=dyn_strandC_a, swing1LimitDegrees=45, swing2LimitDegrees=45, twistLimitDegrees=5}
 {
  bone1Name=dyn_strandB_a_base
  bone2Name=head
  swing1LimitDegrees=45
  swing1Motion=Free
  swing2LimitDegrees=45
  swing2Motion=Free
  twistLimitDegrees=45
  twistMotion=Free
 }
 {bone1Name=dyn_strandB_a, bone2Name=dyn_strandB_a_base, swing1LimitDegrees=45, swing2LimitDegrees=45, twistLimitDegrees=5}
 {bone1Name=dyn_strandB_b, bone2Name=dyn_strandB_a, swing1LimitDegrees=45, swing2LimitDegrees=45, twistLimitDegrees=5}
 {
  bone1Name=dyn_strandMainB_a_base
  bone2Name=head
  swing1LimitDegrees=45
  swing1Motion=Free
  swing2LimitDegrees=45
  swing2Motion=Free
  twistLimitDegrees=45
  twistMotion=Free
 }
 {bone1Name=dyn_strandMainB_a, bone2Name=dyn_strandMainB_a_base, swing1LimitDegrees=20, swing2LimitDegrees=5, twistLimitDegrees=5}
 {bone1Name=dyn_strandMainB_b, bone2Name=dyn_strandMainB_a, swing1LimitDegrees=20, swing2LimitDegrees=20, twistLimitDegrees=5}
 {bone1Name=dyn_strandE_ca, bone2Name=dyn_strandMainB_b, swing1LimitDegrees=20, swing2LimitDegrees=20, twistLimitDegrees=5}
 {bone1Name=dyn_strandE_cb, bone2Name=dyn_strandMainB_b, swing1LimitDegrees=20, swing2LimitDegrees=20, twistLimitDegrees=5}
 {
  bone1Name=dyn_strandA_a_base
  bone2Name=head
  swing1LimitDegrees=45
  swing1Motion=Free
  swing2LimitDegrees=45
  swing2Motion=Free
  twistLimitDegrees=45
  twistMotion=Free
 }
 {bone1Name=dyn_strandA_a, bone2Name=dyn_strandA_a_base, swing1LimitDegrees=25, swing2LimitDegrees=25, twistLimitDegrees=5}
 {bone1Name=dyn_strandA_b, bone2Name=dyn_strandA_a, swing1LimitDegrees=25, swing2LimitDegrees=25, twistLimitDegrees=5}
 {
  bone1Name=clavicle_r
  bone2Name=spine_05
  swing1LimitDegrees=45
  swing1Motion=Free
  swing2LimitDegrees=45
  swing2Motion=Free
  twistLimitDegrees=45
  twistMotion=Free
 }
 {
  bone1Name=dyn_strandD_a_base
  bone2Name=head
  swing1LimitDegrees=45
  swing1Motion=Free
  swing2LimitDegrees=45
  swing2Motion=Free
  twistLimitDegrees=45
  twistMotion=Free
 }
 {bone1Name=dyn_strandD_a, bone2Name=dyn_strandD_a_base, swing1LimitDegrees=35, swing2LimitDegrees=35, twistLimitDegrees=5}
 {bone1Name=dyn_strandD_b, bone2Name=dyn_strandD_a, swing1LimitDegrees=35, swing2LimitDegrees=35, twistLimitDegrees=5}
 {
  bone1Name=dyn_strandF_a_base
  bone2Name=head
  swing1LimitDegrees=45
  swing1Motion=Free
  swing2LimitDegrees=45
  swing2Motion=Free
  twistLimitDegrees=45
  twistMotion=Free
 }
 {bone1Name=dyn_strandF_a, bone2Name=dyn_strandF_a_base, swing1LimitDegrees=15, swing2LimitDegrees=15, twistLimitDegrees=5}
 {bone1Name=dyn_strandF_b, bone2Name=dyn_strandF_a, swing1LimitDegrees=15, swing2LimitDegrees=15, twistLimitDegrees=5}
 {
  bone1Name=dyn_ribbon
  bone2Name=head
  swing1LimitDegrees=45
  swing1Motion=Free
  swing2LimitDegrees=45
  swing2Motion=Free
  twistLimitDegrees=45
  twistMotion=Free
 }
 {bone1Name=dyn_ribbon_l, bone2Name=dyn_ribbon, swing1LimitDegrees=15, swing2LimitDegrees=15, twistLimitDegrees=15}
 {bone1Name=dyn_ribbon_r, bone2Name=dyn_ribbon, swing1LimitDegrees=15, swing2LimitDegrees=15, twistLimitDegrees=25}
 {
  bone1Name=neck_02
  bone2Name=neck_01
  swing1LimitDegrees=45
  swing1Motion=Free
  swing2LimitDegrees=45
  swing2Motion=Free
  twistLimitDegrees=45
  twistMotion=Free
 }
]
