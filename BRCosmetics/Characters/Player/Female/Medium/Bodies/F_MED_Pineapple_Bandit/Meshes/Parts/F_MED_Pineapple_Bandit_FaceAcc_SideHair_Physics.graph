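# Physics asset for the side-hair face accessory of the female medium Pineapple Bandit player character, defining capsule bodies and joints for its hair strands.
# Larger capsules on the upper spine, clavicles, neck and head form the body volume around the hair.
# The main strand chains from the head, and a secondary H strand branches from its second body.
# Body, root and branch-base joints swing and twist freely, while outer strand joints use limited swing and tight twist.
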PhysicsAsset /BRCosmetics/Characters/Player/Female/Medium/Bodies/F_MED_Pineapple_Bandit/Meshes/Parts/F_MED_Pineapple_Bandit_FaceAcc_SideHair_Physics
bodies {
 clavicle_l [{shapeType=Capsule, length=9.25, radius=6.25, center=<3.62 -0.9816 -0.3333>, rotation=<rot 75.9422 0 0>}]
 clavicle_r [{shapeType=Capsule, length=9.25, radius=6.25, center=<-3.6199 0.8332 0.3333>, rotation=<rot 75.9414 0 180>}]
 dyn_strandH_a [{shapeType=Capsule, length=3.6344, radius=2.2469, center=<3.1909 -0.2234 0.3777>, rotation=<rot 0 90.3149 -89.8193>}]
 dyn_strandH_a_base [{shapeType=Capsule, length=0.5, radius=0.5}]
 dyn_strandMainA_a_base [{shapeType=Capsule, length=0.5, radius=0.5}]
 dyn_strandMainA_b [{shapeType=Capsule, length=0.5, radius=0.5}]
 dyn_strandMainA_c [{shapeType=Capsule, length=3.25, radius=1, center=<2.765 0.1943 0>, rotation=<rot 90 0 0>}]
 dyn_strandMainA_d [{shapeType=Capsule, length=5, radius=1, center=<3.3148 0 0>, rotation=<rot 90 0 3.8484>}]
 head [{shapeType=Capsule, length=7, radius=7.5, center=<3.6495 0.5989 0.0001>, rotation=<rot 90 0.0066 0.0066>}]
 neck_01 [{shapeType=Capsule, length=8.25, radius=6.25, center=<-0.3898 1.4176 0>, rotation=<rot 90 0 0>}]
 spine_05 [{shapeType=Capsule, length=13.25, radius=8.25, center=<2.8263 -1.1791 -0.0001>}]
}
constraints [
 {
  bone1Name=head
  bone2Name=neck_01
  swing1LimitDegrees=45
  swing1Motion=Free
  swing2LimitDegrees=45
  swing2Motion=Free
  twistLimitDegrees=45
  twistMotion=Free
 }
 {
  bone1Name=clavicle_l
  bone2Name=spine_05
  swing1LimitDegrees=45
  swing1Motion=Free
  swing2LimitDegrees=45
  swing2Motion=Free
  twistLimitDegrees=45
  twistMotion=Free
 }
 {
  bone1Name=neck_01
  bone2Name=spine_05
  swing1LimitDegrees=45
  swing1Motion=Free
  swing2LimitDegrees=45
  swing2Motion=Free
  twistLimitDegrees=45
  twistMotion=Free
 }
 {
  bone1Name=dyn_strandMainA_a_base
  bone2Name=head
  swing1LimitDegrees=45
  swing1Motion=Free
  swing2LimitDegrees=45
  swing2Motion=Free
  twistLimitDegrees=45
  twistMotion=Free
 }
 {
  bone1Name=clavicle_r
  bone2Name=spine_05
  swing1LimitDegrees=45
  swing1Motion=Free
  swing2LimitDegrees=45
  swing2Motion=Free
  twistLimitDegrees=45
  twistMotion=Free
 }
 {
  bone1Name=dyn_strandMainA_b
  bone2Name=dyn_strandMainA_a_base
  swing1LimitDegrees=45
  swing1Motion=Free
  swing2LimitDegrees=45
  swing2Motion=Free
  twistLimitDegrees=45
  twistMotion=Free
 }
 {bone1Name=dyn_strandMainA_c, bone2Name=dyn_strandMainA_b, swing1LimitDegrees=15, swing2LimitDegrees=15, twistLimitDegrees=5}
 {bone1Name=dyn_strandMainA_d, bone2Name=dyn_strandMainA_c, swing1LimitDegrees=30, swing2LimitDegrees=30, twistLimitDegrees=5}
 {
  bone1Name=dyn_strandH_a_base
  bone2Name=dyn_strandMainA_b
  swing1LimitDegrees=45
  swing1Motion=Free
  swing2LimitDegrees=45
  swing2Motion=Free
  twistLimitDegrees=45
  twistMotion=Free
 }
 {bone1Name=dyn_strandH_a, bone2Name=dyn_strandH_a_base, swing1LimitDegrees=30, swing2LimitDegrees=30, twistLimitDegrees=5}
]
