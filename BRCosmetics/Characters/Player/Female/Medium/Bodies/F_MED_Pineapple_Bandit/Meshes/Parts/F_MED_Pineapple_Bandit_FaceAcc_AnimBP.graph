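# Post-process style AnimBP for the female Pineapple Bandit FaceAcc part, animating its dynamic hair strands in jacket-on and jacket-off variants.
# A copied mesh pose gets weighted multi-bone strand constraints, cached as a base for both variants.
# Each variant adds head-pitch-driven corrective offsets, then RigidBody simulation with its own physics asset plus side-hair physics weighted by left-arm rotation.
# Update logic reads bone rotations and hair meta tags, whose jacketON result selects the output.
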AnimBlueprint /BRCosmetics/Characters/Player/Female/Medium/Bodies/F_MED_Pineapple_Bandit/Meshes/Parts/F_MED_Pineapple_Bandit_FaceAcc_AnimBP
target_skeleton=&/F_MED_Pineapple_Bandit_FaceAcc_Skeleton
graphs [
 AnimGraph {nodes [
   Copy Pose From Mesh() {bUseAttachedParent=true}
   n1 = Get headFwdBkRot() -> (headFwdBkRot:double)
   "Transform (Modify) Bone - Bone: dyn_strandMainA_a_base"(Alpha=@n1, Translation="0,1,0", Rotation="-5,0,0", Scale="1,1,1") {
    rotationMode=BMM_Additive
    rotationSpace=BCS_BoneSpace
    translationMode=BMM_Additive
    translationSpace=BCS_BoneSpace
    alphaBoolBlend {blendOption=Linear}
    alphaScaleBiasClamp {bMapRange=true, inRange {max=0, min=-50}, outRange {max=0, min=1}}
    boneToModify {boneName=dyn_strandMainA_a_base}
    rotation {pitch=-5}
    translation {y=1}
   }
   Output Pose(Result=Blend Poses by bool(
     BlendPose_0=Use cached pose 'jacket_ON_pose'()
     BlendPose_1=Use cached pose 'jacket_OFF_pose'()
     bActiveValue=Get jacketON() -> (jacketON:bool)
    ) {blendType=Linear}) {layerGroup=None}
   n2 = Get MeshToCopy() -> (MeshToCopy:Skeletal Mesh Component Object Reference)
   n3 = "Constraint - dyn_strandMainA_b (PPPPPPPP)"!(
    ComponentPose=Local To Component(LocalPose=Copy Pose From Mesh(SourceMeshComponent=@n2) {bUseAttachedParent=true})
    Alpha=1
    ConstraintWeights_0=0.2006
    ConstraintWeights_1:float=0.0291
    ConstraintWeights_2:float=0.0001
    ConstraintWeights_3:float=0.352
    ConstraintWeights_4:float=0
    ConstraintWeights_5:float=0.0006
    ConstraintWeights_6:float=0.3083
    ConstraintWeights_7:float=0.1092
   ) -> (Pose) {
    alphaBoolBlend {blendOption=Linear}
    boneToModify {boneName=dyn_strandMainA_b}
    constraintSetup [
     {targetBone {boneName=head}}
     {targetBone {boneName=clavicle_l}}
     {targetBone {boneName=spine_04}}
     {targetBone {boneName=spine_05}}
     {targetBone {boneName=upperarm_twist_01_l}}
     {targetBone {boneName=clavicle_r}}
     {targetBone {boneName=neck_02}}
     {targetBone {boneName=neck_01}}
    ]
    constraintWeights [1,1,1,1,1,1,1,1]
   }
   n4 = "Constraint - dyn_strandMainA_c (PPP)"!(ComponentPose=@n3, Alpha=1, ConstraintWeights_0=0.0133, ConstraintWeights_1:float=0.0602, ConstraintWeights_2:float=0.9265) -> (Pose) {
    alphaBoolBlend {blendOption=Linear}
    boneToModify {boneName=dyn_strandMainA_c}
    constraintSetup [{targetBone {boneName=spine_04}}, {targetBone {boneName=spine_05}}, {targetBone {boneName=clavicle_l}}]
    constraintWeights [1,1,1]
   }
   n5 = "Constraint - dyn_strandMainA_d (PPP)"!(ComponentPose=@n4, Alpha=1, ConstraintWeights_0=0.1503, ConstraintWeights_1:float=0.1936, ConstraintWeights_2:float=0.6561) -> (Pose) {
    alphaBoolBlend {blendOption=Linear}
    boneToModify {boneName=dyn_strandMainA_d}
    constraintSetup [{targetBone {boneName=spine_04}}, {targetBone {boneName=spine_05}}, {targetBone {boneName=clavicle_l}}]
    constraintWeights [1,1,1]
   }
   n6 = "Constraint - dyn_strandMainB_b (PPPPPPPP)"!(
    ComponentPose=@n5
    Alpha=1
    ConstraintWeights_0=0.5974
    ConstraintWeights_1:float=0
    ConstraintWeights_2:float=0
    ConstraintWeights_3:float=0.0164
    ConstraintWeights_4:float=0
    ConstraintWeights_5:float=0
    ConstraintWeights_6:float=0.3614
    ConstraintWeights_7:float=0.0247
   ) -> (Pose) {
    alphaBoolBlend {blendOption=Linear}
    boneToModify {boneName=dyn_strandMainB_b}
    constraintSetup [
     {targetBone {boneName=head}}
     {targetBone {boneName=clavicle_l}}
     {targetBone {boneName=spine_04}}
     {targetBone {boneName=spine_05}}
     {targetBone {boneName=upperarm_twist_01_l}}
     {targetBone {boneName=clavicle_r}}
     {targetBone {boneName=neck_02}}
     {targetBone {boneName=neck_01}}
    ]
    constraintWeights [1,1,1,1,1,1,1,1]
   }
   Save cached pose 'BasePose'(Pose=Component To Local(ComponentPose=@n6)) {cacheName=BasePose}
   n7 = Get headFwdBkRot()
   n8 = "Transform (Modify) Bone - Bone: dyn_strandMainA_b"(
    ComponentPose=Local To Component(LocalPose=Use cached pose 'BasePose'())
    Alpha=@n7
    Translation="-2,0.75,1"
    Rotation="0,0,0"
    Scale="1,1,1"
   ) {
    rotationMode=BMM_Additive
    rotationSpace=BCS_BoneSpace
    translationMode=BMM_Additive
    translationSpace=BCS_BoneSpace
    alphaBoolBlend {blendOption=Linear}
    alphaScaleBiasClamp {bMapRange=true, inRange {max=0, min=-50}, outRange {max=0, min=1}}
    boneToModify {boneName=dyn_strandMainA_b}
    translation=<-2 0.75 1>
   }
   n9 = "Transform (Modify) Bone - Bone: dyn_strandMainA_c"(ComponentPose=@n8, Alpha=@n7, Translation="0,0,2", Rotation="-10,0,0", Scale="1,1,1") {
    rotationMode=BMM_Additive
    rotationSpace=BCS_BoneSpace
    translationMode=BMM_Additive
    translationSpace=BCS_BoneSpace
    alphaBoolBlend {blendOption=Linear}
    alphaScaleBiasClamp {bMapRange=true, inRange {max=0, min=-50}, outRange {max=0, min=1}}
    boneToModify {boneName=dyn_strandMainA_c}
    rotation {pitch=-10}
    translation {z=2}
   }
   n10 = "Transform (Modify) Bone - Bone: dyn_strandMainB_a_base"(ComponentPose=@n9, Alpha=@n7, Translation="0,-2,0", Rotation="0,-40,0", Scale="1,1,1") {
    rotationMode=BMM_Additive
    rotationSpace=BCS_BoneSpace
    translationMode=BMM_Additive
    translationSpace=BCS_BoneSpace
    alphaBoolBlend {blendOption=Linear}
    alphaScaleBiasClamp {bMapRange=true, inRange {max=0, min=-50}, outRange {max=0, min=1}}
    boneToModify {boneName=dyn_strandMainB_a_base}
    rotation {yaw=-40}
    translation {y=-2}
   }
   n11 = Get headFwdBkRot()
   n12 = "Transform (Modify) Bone - Bone: dyn_strandC_a_base"(ComponentPose=@n10, Alpha=@n11, Translation="0,-2,0", Rotation="0,-40,0", Scale="1,1,1") {
    rotationMode=BMM_Additive
    rotationSpace=BCS_BoneSpace
    translationMode=BMM_Additive
    translationSpace=BCS_BoneSpace
    alphaBoolBlend {blendOption=Linear}
    alphaScaleBiasClamp {bMapRange=true, inRange {max=0, min=-50}, outRange {max=0, min=1}}
    boneToModify {boneName=dyn_strandC_a_base}
    rotation {yaw=-40}
    translation {y=-2}
   }
   n13 = "Transform (Modify) Bone - Bone: dyn_strandF_a_base"(ComponentPose=@n12, Alpha=@n11, Translation="0,-4,0", Rotation="0,-40,0", Scale="1,1,1") {
    rotationMode=BMM_Additive
    rotationSpace=BCS_BoneSpace
    translationMode=BMM_Additive
    translationSpace=BCS_BoneSpace
    alphaBoolBlend {blendOption=Linear}
    alphaScaleBiasClamp {bMapRange=true, inRange {max=0, min=-50}, outRange {max=0, min=1}}
    boneToModify {boneName=dyn_strandF_a_base}
    rotation {yaw=-40}
    translation {y=-4}
   }
   n14 = "Transform (Modify) Bone - Bone: dyn_strandD_a_base"(ComponentPose=@n13, Alpha=@n11, Translation="0,-2,0", Rotation="0,-40,0", Scale="1,1,1") {
    rotationMode=BMM_Additive
    rotationSpace=BCS_BoneSpace
    translationMode=BMM_Additive
    translationSpace=BCS_BoneSpace
    alphaBoolBlend {blendOption=Linear}
    alphaScaleBiasClamp {bMapRange=true, inRange {max=0, min=-50}, outRange {max=0, min=1}}
    boneToModify {boneName=dyn_strandD_a_base}
    rotation {yaw=-40}
    translation {y=-2}
   }
   n15 = Get headFwdBkRot()
   n16 = "Transform (Modify) Bone - Bone: dyn_strandB_a_base"(ComponentPose=@n14, Alpha=@n15, Translation="0,-2,0", Rotation="0,-40,0", Scale="1,1,1") {
    rotationMode=BMM_Additive
    rotationSpace=BCS_BoneSpace
    translationMode=BMM_Additive
    translationSpace=BCS_BoneSpace
    alphaBoolBlend {blendOption=Linear}
    alphaScaleBiasClamp {bMapRange=true, inRange {max=0, min=-50}, outRange {max=0, min=1}}
    boneToModify {boneName=dyn_strandB_a_base}
    rotation {yaw=-40}
    translation {y=-2}
   }
   n17 = "Transform (Modify) Bone - Bone: dyn_strandH_a_base"(ComponentPose=@n16, Alpha=@n15, Translation="0,1,0", Rotation="10,10,0", Scale="1,1,1") {
    rotationMode=BMM_Additive
    rotationSpace=BCS_BoneSpace
    translationMode=BMM_Additive
    translationSpace=BCS_BoneSpace
    alphaBoolBlend {blendOption=Linear}
    alphaScaleBiasClamp {bMapRange=true, inRange {max=0, min=-50}, outRange {min=0.5}}
    boneToModify {boneName=dyn_strandH_a_base}
    rotation {pitch=10, yaw=10}
    translation {y=1}
   }
   n18 = RigidBody(ComponentPose=@n17) {
    bClampLinearTranslationLimitToRefPose=true
    bOverrideWorldGravity=true
    overridePhysicsAsset=&/F_MED_Pineapple_Bandit_FaceAcc_NoJacket_Physics
    simulationSpace=BaseBoneSpace
    alphaBoolBlend {blendOption=Linear}
    baseBoneRef {boneName=neck_01}
    componentLinearAccScale=<0.35 0.35 0.35>
    componentLinearVelScale=<0.35 0.35 0.35>
    overrideWorldGravity {z=-980}
   }
   n19 = RigidBody(ComponentPose=@n18, Alpha=Get armUp_L() -> (armUp_L:double)) {
    bClampLinearTranslationLimitToRefPose=true
    bOverrideWorldGravity=true
    overridePhysicsAsset=&/F_MED_Pineapple_Bandit_FaceAcc_SideHair_Physics
    simulationSpace=BaseBoneSpace
    alphaBoolBlend {blendOption=Linear}
    alphaScaleBiasClamp {bMapRange=true, inRange {max=20, min=10}, outRange {max=0.7, min=1}}
    baseBoneRef {boneName=neck_01}
    componentLinearAccScale=<0.35 0.35 0.35>
    componentLinearVelScale=<0.35 0.35 0.35>
    overrideWorldGravity {z=-980}
   }
   Save cached pose 'jacket_OFF_pose'(Pose=Component To Local(ComponentPose=@n19)) {cacheName=jacket_OFF_pose}
   n20 = "Constraint - dyn_strandD_b (PPPPPPP)"!(
    ComponentPose=Local To Component(LocalPose=Use cached pose 'BasePose'())
    Alpha=1
    ConstraintWeights_0=0.2274
    ConstraintWeights_1:float=0.0021
    ConstraintWeights_2:float=0
    ConstraintWeights_3:float=0.2917
    ConstraintWeights_4:float=0.0024
    ConstraintWeights_5:float=0.3515
    ConstraintWeights_6:float=0.1249
   ) -> (Pose) {
    alphaBoolBlend {blendOption=Linear}
    boneToModify {boneName=dyn_strandD_b}
    constraintSetup [
     {targetBone {boneName=head}}
     {targetBone {boneName=clavicle_l}}
     {targetBone {boneName=spine_04}}
     {targetBone {boneName=spine_05}}
     {targetBone {boneName=clavicle_r}}
     {targetBone {boneName=neck_02}}
     {targetBone {boneName=neck_01}}
    ]
    constraintWeights [1,1,1,1,1,1,1]
   }
   n21 = "Constraint - dyn_strandF_b (PPPPPPP)"!(
    ComponentPose=@n20
    Alpha=1
    ConstraintWeights_0=0.0568
    ConstraintWeights_1:float=0.0203
    ConstraintWeights_2:float=0
    ConstraintWeights_3:float=0.6846
    ConstraintWeights_4:float=0.0078
    ConstraintWeights_5:float=0.1526
    ConstraintWeights_6:float=0.0779
   ) -> (Pose) {
    alphaBoolBlend {blendOption=Linear}
    boneToModify {boneName=dyn_strandF_b}
    constraintSetup [
     {targetBone {boneName=head}}
     {targetBone {boneName=clavicle_l}}
     {targetBone {boneName=spine_04}}
     {targetBone {boneName=spine_05}}
     {targetBone {boneName=clavicle_r}}
     {targetBone {boneName=neck_02}}
     {targetBone {boneName=neck_01}}
    ]
    constraintWeights [1,1,1,1,1,1,1]
   }
   n22 = "Constraint - dyn_strandMainB_b (PPPPPPP)"!(
    ComponentPose=@n21
    Alpha=1
    ConstraintWeights_0=0.5974
    ConstraintWeights_1:float=0
    ConstraintWeights_2:float=0
    ConstraintWeights_3:float=0.0164
    ConstraintWeights_4:float=0
    ConstraintWeights_5:float=0.3614
    ConstraintWeights_6:float=0.0247
   ) -> (Pose) {
    alphaBoolBlend {blendOption=Linear}
    boneToModify {boneName=dyn_strandMainB_b}
    constraintSetup [
     {targetBone {boneName=head}}
     {targetBone {boneName=clavicle_l}}
     {targetBone {boneName=spine_04}}
     {targetBone {boneName=spine_05}}
     {targetBone {boneName=clavicle_r}}
     {targetBone {boneName=neck_02}}
     {targetBone {boneName=neck_01}}
    ]
    constraintWeights [1,1,1,1,1,1,1]
   }
   n23 = "Constraint - dyn_strandE_ca (PPPPPPP)"!(
    ComponentPose=@n22
    Alpha=1
    ConstraintWeights_0=0.2286
    ConstraintWeights_1:float=0.0014
    ConstraintWeights_2:float=0
    ConstraintWeights_3:float=0.3181
    ConstraintWeights_4:float=0.0028
    ConstraintWeights_5:float=0.3333
    ConstraintWeights_6:float=0.1159
   ) -> (Pose) {
    alphaBoolBlend {blendOption=Linear}
    boneToModify {boneName=dyn_strandE_ca}
    constraintSetup [
     {targetBone {boneName=head}}
     {targetBone {boneName=clavicle_l}}
     {targetBone {boneName=spine_04}}
     {targetBone {boneName=spine_05}}
     {targetBone {boneName=clavicle_r}}
     {targetBone {boneName=neck_02}}
     {targetBone {boneName=neck_01}}
    ]
    constraintWeights [1,1,1,1,1,1,1]
   }
   n24 = "Constraint - dyn_strandE_cb (PPPPPPP)"!(
    ComponentPose=@n23
    Alpha=1
    ConstraintWeights_0=0.2014
    ConstraintWeights_1:float=0.0006
    ConstraintWeights_2:float=0
    ConstraintWeights_3:float=0.3722
    ConstraintWeights_4:float=0.003
    ConstraintWeights_5:float=0.3063
    ConstraintWeights_6:float=0.1165
   ) -> (Pose) {
    alphaBoolBlend {blendOption=Linear}
    boneToModify {boneName=dyn_strandE_cb}
    constraintSetup [
     {targetBone {boneName=head}}
     {targetBone {boneName=clavicle_l}}
     {targetBone {boneName=spine_04}}
     {targetBone {boneName=spine_05}}
     {targetBone {boneName=clavicle_r}}
     {targetBone {boneName=neck_02}}
     {targetBone {boneName=neck_01}}
    ]
    constraintWeights [1,1,1,1,1,1,1]
   }
   n25 = "Transform (Modify) Bone - Bone: dyn_strandMainA_b"(ComponentPose=@n24, Alpha=@n1, Translation="0,1,-1", Rotation="0,10,0", Scale="1,1,1") {
    rotationMode=BMM_Additive
    rotationSpace=BCS_BoneSpace
    translationMode=BMM_Additive
    translationSpace=BCS_BoneSpace
    alphaBoolBlend {blendOption=Linear}
    alphaScaleBiasClamp {bMapRange=true, inRange {max=0, min=-50}, outRange {max=0, min=1}}
    boneToModify {boneName=dyn_strandMainA_b}
    rotation {yaw=10}
    translation {y=1, z=-1}
   }
   n26 = "Transform (Modify) Bone - Bone: dyn_strandMainA_c"(ComponentPose=@n25, Alpha=@n1, Translation="0,0,0", Rotation="-15,-10,0", Scale="1,1,1") {
    rotationMode=BMM_Additive
    rotationSpace=BCS_BoneSpace
    translationMode=BMM_Additive
    translationSpace=BCS_BoneSpace
    alphaBoolBlend {blendOption=Linear}
    alphaScaleBiasClamp {bMapRange=true, inRange {max=0, min=-50}, outRange {max=0, min=1}}
    boneToModify {boneName=dyn_strandMainA_c}
    rotation {pitch=-15, yaw=-10}
   }
   n27 = Get headFwdBkRot()
   n28 = "Transform (Modify) Bone - Bone: dyn_strandH_a_base"(ComponentPose=@n26, Alpha=@n27, Translation="0,0,-2", Rotation="0,-5,0", Scale="1,1,1") {
    rotationMode=BMM_Additive
    rotationSpace=BCS_BoneSpace
    translationMode=BMM_Additive
    translationSpace=BCS_BoneSpace
    alphaBoolBlend {blendOption=Linear}
    alphaScaleBiasClamp {bMapRange=true, inRange {max=0, min=-50}, outRange {max=0, min=1}}
    boneToModify {boneName=dyn_strandH_a_base}
    rotation {yaw=-5}
    translation {z=-2}
   }
   n29 = "Transform (Modify) Bone - Bone: dyn_strandH_a_base"(ComponentPose=@n28, Alpha=@n27, Translation="0,1,0", Rotation="0,0,0", Scale="1,1,1") {
    rotationMode=BMM_Additive
    rotationSpace=BCS_BoneSpace
    translationMode=BMM_Additive
    translationSpace=BCS_BoneSpace
    alphaBoolBlend {blendOption=Linear}
    alphaScaleBiasClamp {bMapRange=true, inRange {max=20, min=-100}}
    boneToModify {boneName=dyn_strandH_a_base}
    translation {y=1}
   }
   n30 = "Transform (Modify) Bone - Bone: dyn_strandMainB_a_base"(ComponentPose=@n29, Alpha=@n27, Translation="0,0,0", Rotation="0,5,0", Scale="1,1,1") {
    rotationMode=BMM_Additive
    rotationSpace=BCS_BoneSpace
    translationMode=BMM_Additive
    translationSpace=BCS_BoneSpace
    alphaBoolBlend {blendOption=Linear}
    alphaScaleBiasClamp {bMapRange=true, inRange {max=0, min=-50}, outRange {max=0, min=1}}
    boneToModify {boneName=dyn_strandMainB_a_base}
    rotation {yaw=5}
   }
   n31 = "Transform (Modify) Bone - Bone: dyn_strandF_b"(ComponentPose=@n30, Alpha=1, Translation="0,0,0", Rotation="0,20,0", Scale="1,1,1") {
    rotationMode=BMM_Additive
    rotationSpace=BCS_BoneSpace
    translationMode=BMM_Additive
    translationSpace=BCS_BoneSpace
    alphaBoolBlend {blendOption=Linear}
    alphaScaleBiasClamp {inRange {max=0, min=-50}, outRange {max=0, min=1}}
    boneToModify {boneName=dyn_strandF_b}
    rotation {yaw=20}
   }
   n32 = RigidBody(ComponentPose=@n31) {
    bClampLinearTranslationLimitToRefPose=true
    bOverrideWorldGravity=true
    overridePhysicsAsset=&/F_MED_Pineapple_Bandit_FaceAcc_Jacket_Physics
    simulationSpace=BaseBoneSpace
    alphaBoolBlend {blendOption=Linear}
    baseBoneRef {boneName=neck_01}
    componentLinearAccScale=<0.35 0.35 0.35>
    componentLinearVelScale=<0.35 0.35 0.35>
    overrideWorldGravity {z=-980}
   }
   n33 = RigidBody(ComponentPose=@n32, Alpha=Get armUp_L()) {
    bClampLinearTranslationLimitToRefPose=true
    bOverrideWorldGravity=true
    overridePhysicsAsset=&F_MED_Pineapple_Bandit_FaceAcc_SideHair_Physics
    simulationSpace=BaseBoneSpace
    alphaBoolBlend {blendOption=Linear}
    alphaScaleBiasClamp {bMapRange=true, inRange {max=15}, outRange {max=0.3, min=0.5}}
    baseBoneRef {boneName=neck_01}
    componentLinearAccScale=<0.2 0.2 0.2>
    componentLinearVelScale=<0.2 0.2 0.2>
    overrideWorldGravity {z=-980}
   }
   Save cached pose 'jacket_ON_pose'(Pose=Component To Local(ComponentPose=@n33)) {cacheName=jacket_ON_pose}
  ]}
 EventGraph {nodes [
   K2Node_Event_0 = CustomEvent() {is=title K2Node_CustomEvent}
   > Set Body_Mesh(Body_Mesh=GetOwnerSkeletalMeshComponent())
   > n1 = Cast To PlayerPawn_Athena(Object=TryGetPawnOwner())
   > n2 = Set Pawn Owner(Pawn Owner:Player Pawn Athena Object Reference=@n1) -> (Output_Get:Player Pawn Athena Object Reference)
   > Set MeshToCopy(MeshToCopy:Skeletal Mesh Component Object Reference=GetSkeletalMeshForPartType(self=@n2, PartType=Head)) -> (Output_Get:Skeletal Mesh Component Object Reference)
   Event BlueprintUpdateAnimation()
   > n3 = Sequence!(execute) -> (then_0, then_1, then_2:Exec, then_3:Exec, then_4:Exec)
   > n4 = Sequence()
   n5 = GetOwnerSkeletalMeshComponent()
   n6 = BreakRotator(InRot=BreakTransform(InTransform=GetDeltaTransformFromRefPose(self=@n5, BoneName=head, BaseName=spine_05)).Rotation)
   > Set headFwdBkRot(headFwdBkRot:double=@n6.Yaw) -> (Output_Get:double)
   > Set headSideToSide(headSideToSide:double=@n6) -> (Output_Get:double)
   n7 = GetDeltaTransformFromRefPose(self=@n5, BoneName=upperarm_l, BaseName=clavicle_l)
   Set armUp_L(execute=@n4.then_1, armUp_L:double=BreakRotator(InRot=BreakTransform(InTransform=@n7).Rotation).Pitch) -> (Output_Get:double)
   n8 = Branch(
    execute=@n3.then_1
    Condition=CharacterItemDefinitionHasAnyMetaTag(Tags=Make Array!("[0]"=(TagName="Cosmetics.MetaTags.Hair.00"), "[1]":Gameplay Tag Structure=(TagName="Cosmetics.MetaTags.Hair.01")) -> (Array))
   )
   > Set jacketON(jacketON:bool=true) -> (Output_Get:bool)
   Set jacketON(execute=@n8.else, jacketON=false)
   TryGetPawnOwner()
  ]}
]
variables [
 armUp_L {type=number}
 Body_Mesh {type=SkeletalMeshComponent}
 headFwdBkRot {type=number}
 headSideToSide {type=number}
 jacketON {type=boolean}
 MeshToCopy {type=SkeletalMeshComponent}
 Pawn Owner {type=PlayerPawn_Athena_C}
]
